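; Prototype TinStack grenade launcher projectile that bounces, then explodes when it stops or first enters water.
; A timer at half the spawn fuse starts a timeline, rate-scaled to that half-fuse, ramping the looping tell's "PreExplo" audio parameter.
; Bounce and stop events record impact orientation and surface, with bounces also driving collision audio and cues.
; Explosion handling is cosmetic-only: sounds, feedback cue and water or default Niagara effect.
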(event Custom|Pre%20Explo%20Audio%20Tell
  ;[1 1 1] Ramp up the audio right before explo
  (Components|Timeline|SetPlayRate (Variables|B_Prj_Ranged_GrenadeLauncher_TinStack|GetPre%20Explo%20Warning) (/ 1.0 (/ (Variables|Default|GetFuseTimer@Spawn) 2.0)))
  (bind (_direction _preexplo) (Timeline|Pre%20Explo%20Warning PlayFromStart)
    (:Update
      (Audio|Parameter|SetFloatParameter (Variables|Default|GetLoopingAudioTell) "PreExplo" _preexplo :!class /Script/AudioExtensions.AudioParameterControllerInterface))))

(event EventBeginPlay
  ;[1 1 1] EVENT -> OnBeginPlay
  (Transformation|SetRelativeScale3D (Variables|Projectile_GrenadeLauncher|GetMesh) "2.000000,2.000000,2.000000")
  (Water|BindEventtoOnEnterWaterEvent (Variables|Water|GetWaterInteractionComponent) (EventDispatchers|CreateEvent :!event Custom|Splash))
  (Networking|CanExecuteCosmeticEvents
    (:True
      (Utilities|Time|SetTimerbyEvent (EventDispatchers|CreateEvent :!event Custom|Pre%20Explo%20Audio%20Tell) (/ (Variables|Default|GetFuseTimer@Spawn) 2.0)))))

(event Custom|Splash
  (params (WaterBody :type Fort%20Water%20Body%20Actor%20Object%20Reference) (WaterInteractionComponent :type Fort%20Water%20Interaction%20Component%20Object%20Reference) (bIsFirstBody :type Boolean))
  ;[0.04342 0.3883 1] Grenade landing in water splash
  (bind _returnvalue (Audio|SpawnSoundAttached "/Game/Sounds/Fort_GamePlay_Sounds/Explosives/Grenade_Projectile_Bounce_Water_Splash_Cue.Grenade_Projectile_Bounce_Water_Splash_Cue" (Variables|Default|GetLoopingAudioTell)))
  (Variables|Default|SetWaterFuseAudioComponent _returnvalue))

(event EventOnStop
  (params (Hit :type Hit%20Result%20Structure%20%28by%20ref%29))
  (bind (_bblockinghit _binitialoverlap _time _distance _location _impactpoint _normal _impactnormal _physmat) (Collision|BreakHitResult Hit))
  (Variables|Default|SetStopLocation _impactpoint) ; Use Loc and Rot for single explosion
  ;[1 1 1] EVENT -> OnStop
  (Variables|Default|SetStopRotX (Math|Rotator|MakeRotfromX _impactnormal))
  (Variables|Default|SetStopRotZ (Math|Rotator|MakeRotfromZ _impactnormal))
  (Utilities|IsValid _physmat
    (:Is%20Valid
      (Variables|Effects|SetStopSurfaceType (Variables|/Script/PhysicsCore.PhysicalMaterial|GetSurfaceType _physmat)))
    (:Is%20Not%20Valid
      _))
  (Rendering|SetVisibility (Variables|Default|GetFlightSystem) false false)
  (if (Utilities|Enum|Equal(Enum) (Variables|Effects|GetStopSurfaceType) "SurfaceType8")
    (Rendering|SetVisibility (Variables|Projectile_GrenadeLauncher|GetMesh) false false)
    (bind _returnvalue (Effects|SpawnEmitteratLocation (Variables|Effects|GetWaterLandDelayPS) (Variables|Default|GetStopLocation) (Variables|Default|GetStopRotX)))
    (Variables|Default|SetWaterLandDelayRipplePSC _returnvalue))
  (CallFunction|FuseTimerMax))

(event Custom|FuseTimerMax
  ;[1 1 1] EVENT -> OnBounce
  ;;[1 1 1] DelayExplosion
  (if (not (Delivery|IsBeingKilled))
    (Components|Activation|SetActive (Variables|Default|GetP_GL_Ribbon_SmokeTrail) false false)
    (Delivery|DoExplosion)))

(event Custom|DelayBeforeExplode_
  ;[1 1 1] EVENT -> OnBounce
  ;;[1 1 1] DelayExplosion
  (Utilities|FlowControl|Delay 0.25)
  (if (not (Delivery|IsBeingKilled))
    (Components|Activation|SetActive (Variables|Default|GetP_GL_Ribbon_SmokeTrail) false false)
    (Delivery|DoExplosion)))

(event EventOnBounce
  (params (Hit :type Hit%20Result%20Structure%20%28by%20ref%29))
  (bind (_bblockinghit _binitialoverlap _time _distance _location _impactpoint _normal _impactnormal _physmat) (Collision|BreakHitResult Hit))
  ;[1 1 1] EVENT -> OnBounce
  (Variables|Default|SetOnBounceHitResult Hit)
  (Utilities|FlowControl|Sequence
    (:then_0
      (Variables|Default|SetStopLocation _impactpoint) ; Use Loc and Rot for single explosion
      (Variables|Default|SetStopRotX (Math|Rotator|MakeRotfromX _impactnormal))
      (Variables|Effects|SetStopSurfaceType (Variables|/Script/PhysicsCore.PhysicalMaterial|GetSurfaceType _physmat)))
    (:then_1
      (bind _hitactor (CallFunction|SpawnBounceFX (Variables|Default|GetOnBounceHitResult)))
      (Variables|Default|SetSpawnBounceFXHitActor _hitactor)
      (Utilities|Casting|CastToPawn (Variables|Default|GetSpawnBounceFXHitActor)))
    (:then_2
      (Fortnite|ShowSoundIndicatorLocal :Location _impactpoint :MaxAudibleDistance 3.5e+03 :IndicatorType "Gunshot" :Instigator self :Tint "(R=0.000000,G=0.000000,B=0.000000,A=0.000000)" :AttenuationMaxRangeMultiplier 1.0 :AttenuationIntensityMultiplier 1.0))))

(event EventOnExploded
  (params (HitActors :type Array%20of%20Actor%20Object%20References) (HitResults :type Array%20of%20Hit%20Result%20Structures))
  (bind _returnvalue (Transformation|GetActorLocation))
  (bind _stoprotx (Variables|Default|GetStopRotX))
  ;[1 1 1] EVENT -> OnExploded
  (Networking|CanExecuteCosmeticEvents
    (:True
      (Development|SetHiddeninGame (Variables|Projectile_GrenadeLauncher|GetMesh) true false)
      (Utilities|FlowControl|Sequence
        (:then_0
          (GameplayCue|ExecuteGameplayCue%28Burst%29Local self (Variables|Default|GetFeedbackCue) (Ability|GameplayCue|MakeGameplayCueParameters 0.0)) ; Cam shake + controller vibration
          (Components|Activation|Deactivate (Variables|Default|GetP_GL_Ribbon_SmokeTrail))
          (Class|BPSurfaceTypeSoundComponent|PlaySound (Variables|Default|GetBP_SurfaceTypeSoundComponent) _returnvalue)
          (Audio|Components|Audio|FadeOut (Variables|Default|GetLoopingAudioTell) 0.3 0.0 "Linear" :!class /Script/Engine.AudioComponent)
          (Fortnite|ShowSoundIndicatorLocal :Location _returnvalue :MaxAudibleDistance 5e+03 :IndicatorType "Gunshot" :Instigator self :Tint "(R=0.000000,G=0.000000,B=0.000000,A=0.000000)" :AttenuationMaxRangeMultiplier 1.0 :AttenuationIntensityMultiplier 1.0)
          (Audio|PlaySoundatLocation (Variables|Sounds|GetGrenade_Explosion_Sound) _returnvalue)
          (bind _waterfuseaudiocomponent (Variables|Default|GetWaterFuseAudioComponent)
            (:then
              (Audio|Components|Audio|FadeOut _waterfuseaudiocomponent 0.2 0.0 "Linear" :!class /Script/Engine.AudioComponent))
            (:else
              _))
          (if (Water|IsinWater (Variables|Water|GetWaterInteractionComponent))
            (Audio|PlaySoundatLocation "/Game/Athena/Sounds/Weapons/Explosions/Expl_Debris_Water_Cue.Expl_Debris_Water_Cue" _returnvalue)))
        (:then_1
          (if (Utilities|Enum|Equal(Enum) (Variables|Default|GetSurfaceType) "SurfaceType8")
            (bind _returnvalue_2 (Niagara|SpawnSystematLocation "/Game/Effects/Explosion_Update/NS_GrenadeLauncher_Explosion_Asteria_Water.NS_GrenadeLauncher_Explosion_Asteria_Water" _returnvalue _stoprotx))
            (else
              (bind _returnvalue_3 (Niagara|SpawnSystematLocation "/Game/Effects/Explosion_Update/NS_GrenadeLauncher_Explosion_Asteria.NS_GrenadeLauncher_Explosion_Asteria" _returnvalue _stoprotx))))
          (Niagara|SetNiagaraVariable(Float) _returnvalue_2 "Global Scale" 0.9))))))

(event EventDestroyed)

(event OnEnterWaterEvent(WaterInteractionComponent)
  (params (WaterBody :type Fort%20Water%20Body%20Actor%20Object%20Reference) (WaterInteractionComponent :type Fort%20Water%20Interaction%20Component%20Object%20Reference) (bIsFirstBody :type Boolean))
  (Utilities|FlowControl|DoOnce false
    (:Completed
      (Audio|Components|Audio|SetSound (Variables|Default|GetLoopingAudioTell) "/Game/Sounds/Fort_Weapons/Launchers/Weapon_LAU_Grenade_Launcher_Athena_RocketWater_Cue.Weapon_LAU_Grenade_Launcher_Athena_RocketWater_Cue")
      (Delivery|DoExplosion))))

(fn SpawnBounceFX
  (params (Hit :type Hit%20Result%20Structure%20%28by%20ref%29))
  (bind (_bblockinghit _binitialoverlap _time _distance _location _impactpoint _normal _impactnormal _physmat _hitactor) (Collision|BreakHitResult Hit))
  (bind _returnvalue (Collision|GetFortPhysicalSurface Hit))
  (bind _outresult (Audio|OnCollision (Variables|Default|GetFortCollisionAudio) _impactpoint (* _impactnormal (Variables|/Script/Engine.MovementComponent|GetVelocity (Variables|Movement|GetProjectileMovementComponent)))))
  (if _outresult
    (GameplayCue|ExecuteGameplayCue%28Burst%29Local self (Variables|Default|GetBounceTag) (Ability|GameplayCue|MakeGameplayCueParameters :NormalizedMagnitude 0.0 :RawMagnitude 0.0 :Location "0,0,0" :Normal "0,0,0" :Instigator (Game|GetInstigator) :GameplayEffectLevel 1 :AbilityLevel 1 :bReplicateLocationWhenUsingMinimalRepProxy false)))
  (Class|BPSurfaceTypeSoundComponent|SetSurfaceType (Variables|Default|GetBP_SurfaceTypeSoundComponent) _returnvalue _hitactor)
  (return _hitactor _returnvalue (Math|Conversions|RotationFromXVector _impactnormal)))

(fn ConstructionScript
  (Effects|Components|ParticleSystem|SetTemplate (Variables|Default|GetRibbonTrail_PSC) (Variables|Effects|GetRibbonTrail_PS))
  (Effects|Components|ParticleSystem|SetTemplate (Variables|Default|GetWaterSplash_PSC) (Variables|Effects|GetBounce_WaterSplash_PS))
  (Effects|Components|ParticleSystem|SetTemplate (Variables|Default|GetBounce_Generic_PSC) (Variables|Effects|GetBounce_Generic_PS))
  (Effects|Components|ParticleSystem|SetTemplate (Variables|Default|GetExplosion_PSC) (Variables|Effects|GetExplosion_PS))
  (Effects|Components|ParticleSystem|SetTemplate (Variables|Default|GetExplosion_Water_PSC) (Variables|Effects|GetExplosion_Water_PS))
  (AddComponent|NODE_AddParticleSystemComponent-1)
  (Audio|SetCollisionSoundatIndex (Variables|Default|GetFortCollisionAudio) 0 (Variables|Sounds|GetGrenade_Bounce_Sound)))
