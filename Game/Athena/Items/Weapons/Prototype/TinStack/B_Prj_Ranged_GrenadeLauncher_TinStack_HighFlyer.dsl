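; HighFlyer projectile for the prototype TinStack grenade launcher, which explodes when it stops on a hit or when its fuse timer expires.
; Timer expiry also sends the instigator a HighFlyerExplode gameplay event with the projectile transform as target data.
; Stopping records the impact, plays bounce feedback and clears the timer; entering water clears it and swaps the looping sound.
; Explosion cosmetics hide flight effects and play cue, surface and water feedback.
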(event EventBeginPlay
  ;[1 1 1] BeginPlay
  (CallFunction|BindExplodeTimer)
  (CallFunction|BindWaterLanding))

(event EventOnStop
  (params (Hit :type Hit%20Result%20Structure%20%28by%20ref%29))
  (bind (_bblockinghit _binitialoverlap _time _distance _location _impactpoint _normal _impactnormal _physmat) (Collision|BreakHitResult Hit))
  ;[1 1 1] OnProjectileStop
  (Variables|HyFlyerBounce|SetBounceHitResult Hit)
  (Variables|HyFlyerBounce|SetBounceLocation _impactpoint)
  (Variables|HyFlyerBounce|SetBounceRotation (Math|Rotator|MakeRotfromX _impactnormal))
  (Variables|HyFlyerBounce|SetBounceSurfaceType (Variables|/Script/PhysicsCore.PhysicalMaterial|GetSurfaceType _physmat))
  (CallFunction|SpawnBounceFX (Variables|HyFlyerBounce|GetBounceHitResult))
  (Fortnite|ShowSoundIndicatorLocal :Location (Variables|HyFlyerBounce|GetBounceLocation) :MaxAudibleDistance 3.5e+03 :IndicatorType "Gunshot" :Instigator self :Tint "(R=0.000000,G=0.000000,B=0.000000,A=0.000000)" :AttenuationMaxRangeMultiplier 1.0 :AttenuationIntensityMultiplier 1.0)
  (Utilities|Time|ClearandInvalidateTimerbyHandle (Variables|HighFlyerDynamic|GetExplodeTimerHandle))
  (Delivery|DoExplosion))

(event Custom|BindWaterLanding
  ;[1 1 1] Water Landing
  (Water|BindEventtoOnEnterWaterEvent (Variables|Water|GetWaterInteractionComponent) (delegate Custom|ExecWaterLanding)))

(event Custom|ExecWaterLanding
  (params (WaterBody :type Fort%20Water%20Body%20Actor%20Object%20Reference) (WaterInteractionComponent :type Fort%20Water%20Interaction%20Component%20Object%20Reference) (bIsFirstBody :type Boolean))
  ;[1 1 1] Water Landing
  (Utilities|Time|ClearandInvalidateTimerbyHandle (Variables|HighFlyerDynamic|GetExplodeTimerHandle))
  (Utilities|FlowControl|DoOnce false
    (:Completed
      (Audio|Components|Audio|SetSound (Variables|Default|GetLoopingAudioTell) "/Game/Sounds/Fort_Weapons/Launchers/Weapon_LAU_Grenade_Launcher_Athena_RocketWater_Cue.Weapon_LAU_Grenade_Launcher_Athena_RocketWater_Cue"))))

(event Custom|BindExplodeTimer
  ;[1 1 1] ExplosionTimer
  (bind _returnvalue (Utilities|Time|SetTimerbyEvent (delegate Custom|ExecExplodeTimer) (Variables|HighFlyerDynamic|GetExplodeTime)))
  (Variables|HighFlyerDynamic|SetExplodeTimerHandle _returnvalue))

(event Custom|ExecExplodeTimer
  (bind _returnvalue (Game|GetInstigator))
  ;[1 1 1] ExplosionTimer
  (bind _asfort_player_pawn (Utilities|Casting|CastToFortPlayerPawn _returnvalue))
  (Ability|SendGameplayEventtoActor _asfort_player_pawn "(TagName=\"Event.TinStack.Trigger.HighFlyerExplode\")" (Utilities|Struct|MakeGameplayEventData "(TagName=\"Event.TinStack.Trigger.HighFlyerExplode\")" _returnvalue _returnvalue 0 0 "()" "(GameplayTags=)" "(GameplayTags=)" 0.0 (Ability|TargetData|AbilityTargetDatafromLocations (Utilities|Struct|MakeGameplayAbilityTargetingLocationInfo self (Variables|Projectile_GrenadeLauncher|GetMesh) 0 (Transformation|GetActorTransform) "None" "ActorTransform") (Utilities|Struct|MakeGameplayAbilityTargetingLocationInfo self (Variables|Projectile_GrenadeLauncher|GetMesh) 0 (Transformation|GetActorTransform) "None" "ActorTransform"))))
  (Delivery|DoExplosion))

(event EventOnExploded
  (params (HitActors :type Array%20of%20Actor%20Object%20References) (HitResults :type Array%20of%20Hit%20Result%20Structures))
  ;[1 1 1] Visual Explosino Effects
  (Networking|CanExecuteCosmeticEvents
    (:True
      (CallFunction|HideVisibleFlightProjectileComponents)
      (GameplayCue|ExecuteGameplayCue%28Burst%29Local self (Variables|ProjectileGrenadeLauncher|GetFeedbackCue) (Ability|GameplayCue|MakeGameplayCueParameters 0.0)) ; Cam shake + controller vibration
      (CallFunction|PlayExposionSounds)
      (CallFunction|AttemptExecuteWaterLanding (Water|IsinWater (Variables|Water|GetWaterInteractionComponent))))))

(fn AttemptExecuteWaterLanding
  (params (Condition :type Boolean))
  (if Condition
    (Audio|PlaySoundatLocation "/Game/Athena/Sounds/Weapons/Explosions/Expl_Debris_Water_Cue.Expl_Debris_Water_Cue" (Transformation|GetActorLocation))
    (if (Utilities|Enum|Equal(Enum) (Variables|HyFlyerBounce|GetBounceSurfaceType) "SurfaceType8")
      (Rendering|SetVisibility (Variables|Projectile_GrenadeLauncher|GetMesh) false false)
      (bind _returnvalue (Effects|SpawnEmitteratLocation (Variables|Effects|GetWaterLandDelayPS) (Variables|HyFlyerBounce|GetBounceLocation) (Variables|HyFlyerBounce|GetBounceRotation)))
      (Variables|Default|SetWaterLandDelayRipplePSC _returnvalue))))

(fn HideVisibleFlightProjectileComponents
  (Components|Activation|SetActive (Variables|Default|GetP_GL_Ribbon_SmokeTrail) false false)
  (Rendering|SetVisibility (Variables|Default|GetFlightSystem) false false)
  (Components|Activation|Deactivate (Variables|Default|GetP_GL_Ribbon_SmokeTrail)))

(fn PlayExposionSounds
  (Class|BPSurfaceTypeSoundComponent|PlaySound (Variables|Default|GetBP_SurfaceTypeSoundComponent) (Transformation|GetActorLocation))
  (Audio|Components|Audio|FadeOut (Variables|Default|GetLoopingAudioTell) 0.3 0.0 "Linear" :!class /Script/Engine.AudioComponent)
  (Fortnite|ShowSoundIndicatorLocal :Location (Transformation|GetActorLocation) :MaxAudibleDistance 5e+03 :IndicatorType "Gunshot" :Instigator self :Tint "(R=0.000000,G=0.000000,B=0.000000,A=0.000000)" :AttenuationMaxRangeMultiplier 1.0 :AttenuationIntensityMultiplier 1.0)
  (Audio|PlaySoundatLocation (Variables|Sounds|GetGrenade_Explosion_Sound) (Transformation|GetActorLocation))
  (bind _waterfuseaudiocomponent (Variables|Default|GetWaterFuseAudioComponent)
    (:then
      (Audio|Components|Audio|FadeOut _waterfuseaudiocomponent 0.2 0.0 "Linear" :!class /Script/Engine.AudioComponent))
    (:else)))

(fn SpawnBounceFX
  (params (Hit :type Hit%20Result%20Structure%20%28by%20ref%29))
  (bind (_bblockinghit _binitialoverlap _time _distance _location _impactpoint _normal _impactnormal _physmat _hitactor) (Collision|BreakHitResult Hit))
  (bind _returnvalue (Collision|GetFortPhysicalSurface Hit))
  (bind _outresult (Audio|OnCollision (Variables|Default|GetFortCollisionAudio) _impactpoint (* _impactnormal (Variables|/Script/Engine.MovementComponent|GetVelocity (Variables|Movement|GetProjectileMovementComponent)))))
  (if _outresult
    (GameplayCue|ExecuteGameplayCue%28Burst%29Local self (Variables|HyFlyerBounce|GetBounceTag) (Ability|GameplayCue|MakeGameplayCueParameters :NormalizedMagnitude 0.0 :RawMagnitude 0.0 :Location "0,0,0" :Normal "0,0,0" :Instigator (Game|GetInstigator) :GameplayEffectLevel 1 :AbilityLevel 1 :bReplicateLocationWhenUsingMinimalRepProxy false)))
  (Class|BPSurfaceTypeSoundComponent|SetSurfaceType (Variables|Default|GetBP_SurfaceTypeSoundComponent) _returnvalue _hitactor)
  (return _hitactor _returnvalue (Math|Conversions|RotationFromXVector _impactnormal)))

(fn ConstructionScript
  (Effects|Components|ParticleSystem|SetTemplate (Variables|Default|GetRibbonTrail_PSC) (Variables|Effects|GetRibbonTrail_PS))
  (Effects|Components|ParticleSystem|SetTemplate (Variables|Default|GetWaterSplash_PSC) (Variables|HyFlyerBounce|GetBounce_WaterSplash_PS))
  (Effects|Components|ParticleSystem|SetTemplate (Variables|Default|GetBounce_Generic_PSC) (Variables|HyFlyerBounce|GetBounce_Generic_PS))
  (Effects|Components|ParticleSystem|SetTemplate (Variables|Default|GetExplosion_PSC) (Variables|Effects|GetExplosion_PS))
  (Effects|Components|ParticleSystem|SetTemplate (Variables|Default|GetExplosion_Water_PSC) (Variables|Effects|GetExplosion_Water_PS))
  (AddComponent|NODE_AddParticleSystemComponent-1)
  (Audio|SetCollisionSoundatIndex (Variables|Default|GetFortCollisionAudio) 0 (Variables|Sounds|GetGrenade_Bounce_Sound)))
