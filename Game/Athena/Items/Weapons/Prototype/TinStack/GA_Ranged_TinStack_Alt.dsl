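; Gameplay ability for the alt behavior of the prototype TinStack ranged weapon, holding a crouching-active gameplay effect on the owner.
; On activation it starts waiting for the crouch gameplay tag to be removed, applies the effect and stores its handle.
; Tag removal ends the ability, and the end event removes the effect by that handle.
; What triggers activation is not defined in this graph.
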(event Ability|EventActivateAbility
  (Ability|Tasks|WaitGameplayTagRemove (Variables|TinStack_AltDefaults|GetEventTagGameplayActionCrouch) 0 true
    (:then
      (bind _returnvalue (Ability|ApplyGameplayEffectSpecToOwner (Ability|MakeOutgoingGameplayEffectSpec (Variables|TinStackAltDefaults|GetGE_CrouchingActive))))
      (Variables|TinStackAltDefaults|SetEffectSpecCrouchHandle _returnvalue))
    (:Removed
      (Ability|EndAbility))))

(event Ability|EventOnEndAbility
  (params (bWasCancelled :type Boolean))
  (Ability|RemoveGameplayEffectFromOwnerWithHandle (Variables|TinStackAltDefaults|GetEffectSpecCrouchHandle)))
